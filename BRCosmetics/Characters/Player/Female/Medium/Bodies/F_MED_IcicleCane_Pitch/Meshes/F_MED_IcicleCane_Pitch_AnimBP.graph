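# Post-process-style body AnimBlueprint for the IcicleCane Pitch female medium character cosmetic, adding secondary motion to a copied pose.
# The attached parent's pose feeds a Control Rig that receives emote and skydive flags.
# AnimDynamics then springs dyn_gun relative to the right thigh and simulates both dyn_torso bones after a small additive offset.
# Each torso bone gets two stacked passes: angular spring, then rotation-locked linear spring with lighter gravity.
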AnimBlueprint /BRCosmetics/Characters/Player/Female/Medium/Bodies/F_MED_IcicleCane_Pitch/Meshes/F_MED_IcicleCane_Pitch_AnimBP
target_skeleton=&/F_MED_IcicleCane_Pitch_Skeleton
graphs [
 AnimGraph {nodes [
   n1 = Control Rig!(Source=Copy Pose From Mesh() {bUseAttachedParent=true}, Alpha, Emotes:bool=Get bIsPlayingEmote(), Skydive:bool=Get bIsSkyDiving()) -> (Pose) {controlRigAssetReference {blueprintRigClass=&/F_MED_IcicleCane_Pitch_CtrlRig.F_MED_IcicleCane_Pitch_CtrlRig_C}}
   n2 = Anim Dynamics - Bone: dyn_gun!(ComponentPose=Local To Component(LocalPose=@n1), Alpha=1) -> (Pose) {
    angularDampingOverride=0.999
    angularSpringConstant=3000
    bAngularSpring=true
    bOverrideAngularDamping=true
    simulationSpace=BoneRelative
    boundBone {boneName=dyn_gun}
    physicsBodyDefinitions [{
      boundBone {boneName=dyn_gun}
      boxExtents=<20 8 10>
      constraintSetup {angularLimitsMax=<2 15 2>, angularLimitsMin=<-2 -15 -2>, angularTarget=<1 0 0>}
      localJointOffset=<0.2 0 0>
     }]
    relativeSpaceBone {boneName=thigh_r}
   }
   n3 = "Transform (Modify) Bone - Bone: dyn_torso_l"(ComponentPose=@n2, Alpha=1, Translation="0,0,1", Rotation="0,0,0", Scale="1,1,1") {translationMode=BMM_Additive, translationSpace=BCS_BoneSpace, boneToModify {boneName=dyn_torso_l}, translation {z=1}}
   n4 = "Transform (Modify) Bone - Bone: dyn_torso_r"(ComponentPose=@n3, Alpha=1, Translation="0,0,1", Rotation="0,0,0", Scale="1,1,1") {translationMode=BMM_Additive, translationSpace=BCS_BoneSpace, boneToModify {boneName=dyn_torso_r}, translation {z=1}}
   n5 = Anim Dynamics - Bone: dyn_torso_r!(ComponentPose=@n4, Alpha=1, GravityOverride:Vector="0,0,-980", ExternalForce="0,0,0") -> (Pose) {
    angularBiasOverride=1
    angularDampingOverride=1
    angularSpringConstant=5000
    bAngularSpring=true
    bGravityOverrideInSimSpace=true
    bOverrideAngularBias=true
    bOverrideAngularDamping=true
    bUseGravityOverride=true
    linearDampingOverride=0.9999
    linearSpringConstant=100
    simulationSpace=RootRelative
    boundBone {boneName=dyn_torso_r}
    componentAppliedLinearAccClamp=<2 2 2>
    gravityOverride {z=-980}
    physicsBodyDefinitions [{
      boundBone {boneName=dyn_torso_r}
      boxExtents=<21 21 21>
      constraintSetup {angularLimitsMax=<0 1 3.6>, angularLimitsMin=<0 -5 -3.6>, angularTarget=<1 0 0>}
      localJointOffset=<2 0 0>
     }]
    relativeSpaceBone {boneName=spine_04}
    simSpaceSettings {maxAngularAcceleration=45, maxAngularVelocity=8, simSpaceAngularAlpha=0.25}
   }
   n6 = Anim Dynamics - Bone: dyn_torso_l!(ComponentPose=@n5, Alpha=1, GravityOverride:Vector="0,0,-980", ExternalForce="0,0,0") -> (Pose) {
    angularBiasOverride=1
    angularDampingOverride=1
    angularSpringConstant=5000
    bAngularSpring=true
    bGravityOverrideInSimSpace=true
    bOverrideAngularBias=true
    bOverrideAngularDamping=true
    bUseGravityOverride=true
    linearDampingOverride=0.9999
    linearSpringConstant=100
    simulationSpace=RootRelative
    boundBone {boneName=dyn_torso_l}
    componentAppliedLinearAccClamp=<2 2 2>
    gravityOverride {z=-980}
    physicsBodyDefinitions [{
      boundBone {boneName=dyn_torso_l}
      boxExtents=<20 20 20>
      constraintSetup {angularLimitsMax=<0 1 3.6>, angularLimitsMin=<0 -5 -3.6>, angularTarget=<1 0 0>}
      localJointOffset=<2 0 0>
     }]
    relativeSpaceBone {boneName=spine_04}
    simSpaceSettings {maxAngularAcceleration=45, maxAngularVelocity=8, simSpaceAngularAlpha=0.25}
   }
   n7 = Anim Dynamics - Bone: dyn_torso_r(ComponentPose=@n6, Alpha=1, GravityOverride="0,0,-200", ExternalForce="0,0,0") {
    angularBiasOverride=1
    angularDampingOverride=0.99
    angularSpringConstant=10000
    bGravityOverrideInSimSpace=true
    bLinearSpring=true
    bOverrideAngularBias=true
    bOverrideLinearDamping=true
    bUseGravityOverride=true
    linearDampingOverride=1
    linearSpringConstant=180
    simulationSpace=RootRelative
    boundBone {boneName=dyn_torso_r}
    componentAppliedLinearAccClamp=<5 5 5>
    gravityOverride {z=-200}
    physicsBodyDefinitions [{
      boundBone {boneName=dyn_torso_r}
      boxExtents=<70 70 70>
      constraintSetup {angularLimitsMax=<0 0 0>, angularLimitsMin=<0 0 0>, angularTarget=<1 0 0>, linearAxesMax=<0 0.35 1>, linearAxesMin=<0 -0.35 -0.5>}
      localJointOffset=<0.2 0 0>
     }]
    relativeSpaceBone {boneName=spine_04}
    simSpaceSettings {maxAngularAcceleration=50, maxAngularVelocity=10, simSpaceAngularAlpha=0.1}
   }
   n8 = Anim Dynamics - Bone: dyn_torso_l(ComponentPose=@n7, Alpha=1, GravityOverride="0,0,-200", ExternalForce="0,0,0") {
    angularBiasOverride=1
    angularDampingOverride=0.99
    angularSpringConstant=10000
    bGravityOverrideInSimSpace=true
    bLinearSpring=true
    bOverrideAngularBias=true
    bOverrideLinearDamping=true
    bUseGravityOverride=true
    linearDampingOverride=1
    linearSpringConstant=180
    simulationSpace=RootRelative
    boundBone {boneName=dyn_torso_l}
    componentAppliedLinearAccClamp=<5 5 5>
    gravityOverride {z=-200}
    physicsBodyDefinitions [{
      boundBone {boneName=dyn_torso_l}
      boxExtents=<80 80 80>
      constraintSetup {angularLimitsMax=<0 0 0>, angularLimitsMin=<0 0 0>, angularTarget=<1 0 0>, linearAxesMax=<0 0.35 1>, linearAxesMin=<0 -0.35 -0.5>}
      localJointOffset=<0.2 0 0>
     }]
    relativeSpaceBone {boneName=spine_04}
    simSpaceSettings {maxAngularAcceleration=50, maxAngularVelocity=10, simSpaceAngularAlpha=0.1}
   }
   Output Pose(Result=Component To Local(ComponentPose=@n8)) {layerGroup=None}
  ]}
 EventGraph {nodes [Event BlueprintUpdateAnimation()]}
]
variables []
