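# Physics asset for the F_MED_IcicleCane_Pitch face accessory's backpack variant, providing capsule bodies and joints for dynamic hair on a female character.
# Center, left, right and branching front hair chains hang from head-linked deform bones with limited swing and twist.
# Cross-constraints tie side chains to the center at each segment, keeping the hair together.
# Torso, head, neck and upper-arm capsules form body volumes, with oversized offset spine capsules likely approximating backpack bulk.
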PhysicsAsset /BRCosmetics/Characters/Player/Female/Medium/Bodies/F_MED_IcicleCane_Pitch/Meshes/Parts/F_MED_IcicleCane_Pitch_FaceAcc_Backpack_Physics
bodies {
 clavicle_l [{
   shapeType=Capsule
   length=9.3767
   radius=5.6234
   shapeName=clavicle_l_capsule
   center=<4.4774 -0.3141 -1.5367>
   rotation=<rot 82.1518 -0.0003 179.9997>
  }]
 clavicle_r [{
   shapeType=Capsule
   length=9.3767
   radius=5.6234
   shapeName=clavicle_r_capsule
   center=<-4.4774 0.3141 1.5367>
   rotation=<rot 82.1518 -0.0003 -0.0003>
  }]
 deform_hair [{shapeType=Capsule, length=0.505, radius=0.505, shapeName=deform_hair_capsule, rotation=<rot 0 -90 0>}]
 deform_hair_f [{shapeType=Capsule, length=1.3071, radius=0.505, shapeName=deform_hair_f_capsule, rotation=<rot 0 -90 0>}]
 deform_hair_l [{shapeType=Capsule, length=0.505, radius=0.505, shapeName=deform_hair_l_capsule, rotation=<rot 0 -90 0>}]
 deform_hair_r [{shapeType=Capsule, length=0.505, radius=0.505, shapeName=deform_hair_r_capsule, rotation=<rot 0 -90 0>}]
 dyn_hair [{shapeType=Capsule, length=10, radius=3, shapeName=dyn_hair_capsule, center=<5.5783 0 0>, rotation=<rot 90 0 0>}]
 dyn_hair_1 [{shapeType=Capsule, length=12, radius=3, shapeName=dyn_hair_1_capsule, center=<5.5783 0 0>, rotation=<rot 90 0 0>}]
 dyn_hair_1_f [{shapeType=Capsule, length=5.8021, radius=3, shapeName=dyn_hair_1_f_capsule, center=<2.6888 -1.4405 0>, rotation=<rot 90 0 0>}]
 dyn_hair_1_l [{shapeType=Capsule, length=12, radius=3, shapeName=dyn_hair_1_l_capsule, center=<5.5783 0 0>, rotation=<rot 90 0 0>}]
 dyn_hair_1_r [{shapeType=Capsule, length=12, radius=3, shapeName=dyn_hair_1_r_capsule, center=<5.5783 0 0>, rotation=<rot 90 0 0>}]
 dyn_hair_2 [{shapeType=Capsule, length=13.423, radius=2.4625, shapeName=dyn_hair_2_capsule, center=<5.9959 0 0>, rotation=<rot 90 0 0>}]
 dyn_hair_2_f [{shapeType=Capsule, length=6.2994, radius=2.5027, shapeName=dyn_hair_2_f_capsule, center=<2.6888 -1.4405 0>, rotation=<rot 90 0 0>}]
 dyn_hair_2_l [{shapeType=Capsule, length=12.8855, radius=3, shapeName=dyn_hair_2_l_capsule, center=<5.9959 0 0>, rotation=<rot 90 0 0>}]
 dyn_hair_2_r [{shapeType=Capsule, length=12.8855, radius=3, shapeName=dyn_hair_2_r_capsule, center=<5.9959 0 0>, rotation=<rot 90 0 0>}]
 dyn_hair_3 [{shapeType=Capsule, length=13.7668, radius=3, shapeName=dyn_hair_3_capsule, center=<6.703 0 0>, rotation=<rot 90 0 0>}]
 dyn_hair_3_f [{shapeType=Capsule, length=6.7717, radius=2.0304, shapeName=dyn_hair_3_f_capsule, center=<2.6888 -1.4405 0>, rotation=<rot 90 0 0>}]
 dyn_hair_3_l [{shapeType=Capsule, length=13.7668, radius=3, shapeName=dyn_hair_3_l_capsule, center=<6.703 0 0>, rotation=<rot 90 0 0>}]
 dyn_hair_3_r [{shapeType=Capsule, length=13.7668, radius=3, shapeName=dyn_hair_3_r_capsule, center=<6.703 0 0>, rotation=<rot 90 0 0>}]
 dyn_hair_4 [{shapeType=Capsule, length=13.7668, radius=3, shapeName=dyn_hair_4_capsule, center=<6.3412 0 0>, rotation=<rot 90 0 0>}]
 dyn_hair_4_f [{shapeType=Capsule, length=6.2994, radius=2.5027, shapeName=dyn_hair_4_f_capsule, center=<2.6888 -1.4405 0>, rotation=<rot 90 0 0>}]
 dyn_hair_4_l [{shapeType=Capsule, length=13.7668, radius=3, shapeName=dyn_hair_4_l_capsule, center=<6.3412 0 0>, rotation=<rot 90 0 0>}]
 dyn_hair_4_r [{shapeType=Capsule, length=13.7668, radius=3, shapeName=dyn_hair_4_r_capsule, center=<6.3412 0 0>, rotation=<rot 90 0 0>}]
 dyn_hair_5_f [{shapeType=Capsule, length=8.516, radius=2.0304, shapeName=dyn_hair_5_f_capsule, center=<2.6888 -1.4405 0>, rotation=<rot 90 0 0>}]
 dyn_hair_6_f [{shapeType=Capsule, length=6.6731, radius=1, shapeName=dyn_hair_6_f_capsule, center=<3.2811 0 0>, rotation=<rot 90 0 0>}]
 dyn_hair_7_f [{shapeType=Capsule, length=3.3878, radius=1.1872, shapeName=dyn_hair_7_f_capsule, center=<3.853 0 0>, rotation=<rot 90 0 0>}]
 dyn_hair_f [{shapeType=Capsule, length=5.8021, radius=3, shapeName=dyn_hair_f_capsule, center=<2.6888 -1.4405 0>, rotation=<rot 90 0 0>}]
 dyn_hair_l [{shapeType=Capsule, length=10, radius=3, shapeName=dyn_hair_l_capsule, center=<5.5783 0 0>, rotation=<rot 90 0 0>}]
 dyn_hair_r [{shapeType=Capsule, length=10, radius=3, shapeName=dyn_hair_r_capsule, center=<5.5783 0 0>, rotation=<rot 90 0 0>}]
 head [
  {shapeType=Capsule, length=10.1372, radius=6.9269, center=<1.7277 2.6597 0>, rotation=<rot 90 -34.451 0>}
  {shapeType=Capsule, length=1.8253, radius=7.5486, shapeName=head_capsule, center=<5.8522 0 0>}
 ]
 neck_01 [{shapeType=Capsule, length=8.5081, radius=4.236, shapeName=neck_01_capsule, center=<4.4087 0 0>, rotation=<rot 90 0 0>}]
 neck_02 [{shapeType=Capsule, length=5.2408, radius=4.7592, shapeName=neck_02_capsule, center=<3.0831 0 0>, rotation=<rot 90 0 0>}]
 pelvis [{shapeType=Capsule, length=6.4284, radius=13.5716, shapeName=pelvis_capsule, center=<-4.0509 0 0>, rotation=<rot 0 -90 0>}]
 spine_01 [{shapeType=Capsule, length=8.7306, radius=8.7167, shapeName=spine_01_capsule, rotation=<rot 0 -90 0>}]
 spine_02 [{shapeType=Capsule, length=9.4229, radius=6.8299, shapeName=spine_02_capsule, center=<0.4025 0 0>, rotation=<rot 0 -90 0>}]
 spine_03 [
  {shapeType=Capsule, length=11.6973, radius=6.7359, shapeName=spine_03_capsule, center=<1.1021 0 0>, rotation=<rot 0 -90 0>}
  {shapeType=Capsule, length=24.9308, radius=21.412, shapeName=spine_03_capsule, center=<3.3752 -35.8218 0>}
 ]
 spine_04 [
  {shapeType=Capsule, length=10.6277, radius=8.2352, shapeName=spine_04_capsule, center=<1.4616 0 0>}
  {shapeType=Capsule, length=27.2612, radius=21.0451, shapeName=spine_04_capsule, center=<6.3074 19.0837 0>}
  {shapeType=Capsule, length=32.7083, radius=18.964, shapeName=spine_04_capsule, center=<0 -35.1465 0>}
 ]
 spine_05 [
  {shapeType=Capsule, length=11.0144, radius=8.9856, shapeName=spine_05_capsule, center=<1.4616 0 0>, rotation=<rot 0 -90 0>}
  {shapeType=Capsule, length=0, radius=0, shapeName=spine_05_capsule, center=<0 -39.0477 0>, rotation=<rot 90 0 0>}
  {shapeType=Capsule, length=33.3045, radius=19.6633, shapeName=spine_05_capsule, center=<10.4046 -32.5585 0>}
 ]
 upperarm_l [{shapeType=Capsule, length=25.2216, radius=4.5398, shapeName=upperarm_l_capsule, center=<12.4412 0 0>, rotation=<rot 90 180 0>}]
 upperarm_r [{shapeType=Capsule, length=25.2216, radius=4.5398, shapeName=upperarm_r_capsule, center=<-12.4412 0 0>, rotation=<rot 90 0 0>}]
}
constraints [
 {
  bone1Name=head
  bone2Name=neck_02
  swing1LimitDegrees=45
  swing1Motion=Free
  swing2LimitDegrees=45
  swing2Motion=Free
  twistLimitDegrees=45
  twistMotion=Free
 }
 {
  bone1Name=neck_01
  bone2Name=spine_05
  swing1LimitDegrees=45
  swing1Motion=Free
  swing2LimitDegrees=45
  swing2Motion=Free
  twistLimitDegrees=45
  twistMotion=Free
 }
 {
  bone1Name=spine_05
  bone2Name=spine_04
  swing1LimitDegrees=45
  swing1Motion=Free
  swing2LimitDegrees=45
  swing2Motion=Free
  twistLimitDegrees=45
  twistMotion=Free
 }
 {
  bone1Name=deform_hair
  bone2Name=head
  swing1LimitDegrees=45
  swing1Motion=Free
  swing2LimitDegrees=45
  swing2Motion=Free
  twistLimitDegrees=45
  twistMotion=Free
 }
 {bone1Name=dyn_hair, bone2Name=deform_hair, swing1LimitDegrees=30, swing2LimitDegrees=25, twistLimitDegrees=5}
 {bone1Name=dyn_hair_1, bone2Name=dyn_hair, swing1LimitDegrees=35, swing2LimitDegrees=25, twistLimitDegrees=5}
 {bone1Name=dyn_hair_2, bone2Name=dyn_hair_1, swing1LimitDegrees=45, swing2LimitDegrees=25, twistLimitDegrees=5}
 {bone1Name=dyn_hair_3, bone2Name=dyn_hair_2, swing1LimitDegrees=45, swing2LimitDegrees=25, twistLimitDegrees=5}
 {bone1Name=dyn_hair_4, bone2Name=dyn_hair_3, swing1LimitDegrees=45, swing2LimitDegrees=25, twistLimitDegrees=5}
 {
  bone1Name=deform_hair_f
  bone2Name=head
  swing1LimitDegrees=45
  swing1Motion=Free
  swing2LimitDegrees=45
  swing2Motion=Free
  twistLimitDegrees=45
  twistMotion=Free
 }
 {bone1Name=dyn_hair_f, bone2Name=deform_hair_f, swing1LimitDegrees=5, swing2LimitDegrees=5, twistLimitDegrees=5}
 {bone1Name=dyn_hair_1_f, bone2Name=dyn_hair_f, swing1LimitDegrees=10, swing2LimitDegrees=10, twistLimitDegrees=5}
 {bone1Name=dyn_hair_2_f, bone2Name=dyn_hair_1_f, swing1LimitDegrees=35, swing2LimitDegrees=15, twistLimitDegrees=5}
 {bone1Name=dyn_hair_3_f, bone2Name=dyn_hair_2_f, swing1LimitDegrees=35, swing2LimitDegrees=25, twistLimitDegrees=5}
 {bone1Name=dyn_hair_4_f, bone2Name=dyn_hair_1_f, swing1LimitDegrees=35, swing2LimitDegrees=15, twistLimitDegrees=5}
 {bone1Name=dyn_hair_5_f, bone2Name=dyn_hair_4_f, swing1LimitDegrees=35, swing2LimitDegrees=25, twistLimitDegrees=5}
 {
  bone1Name=deform_hair_l
  bone2Name=head
  swing1LimitDegrees=45
  swing1Motion=Free
  swing2LimitDegrees=45
  swing2Motion=Free
  twistLimitDegrees=45
  twistMotion=Free
 }
 {bone1Name=dyn_hair_l, bone2Name=deform_hair_l, swing1LimitDegrees=30, swing2LimitDegrees=25, twistLimitDegrees=5}
 {bone1Name=dyn_hair_1_l, bone2Name=dyn_hair_l, swing1LimitDegrees=35, swing2LimitDegrees=25, twistLimitDegrees=5}
 {bone1Name=dyn_hair_2_l, bone2Name=dyn_hair_1_l, swing1LimitDegrees=45, swing2LimitDegrees=25, twistLimitDegrees=5}
 {bone1Name=dyn_hair_3_l, bone2Name=dyn_hair_2_l, swing1LimitDegrees=45, swing2LimitDegrees=25, twistLimitDegrees=5}
 {bone1Name=dyn_hair_4_l, bone2Name=dyn_hair_3_l, swing1LimitDegrees=45, swing2LimitDegrees=25, twistLimitDegrees=5}
 {
  bone1Name=deform_hair_r
  bone2Name=head
  swing1LimitDegrees=45
  swing1Motion=Free
  swing2LimitDegrees=45
  swing2Motion=Free
  twistLimitDegrees=45
  twistMotion=Free
 }
 {bone1Name=dyn_hair_r, bone2Name=deform_hair_r, swing1LimitDegrees=30, swing2LimitDegrees=25, twistLimitDegrees=5}
 {bone1Name=dyn_hair_1_r, bone2Name=dyn_hair_r, swing1LimitDegrees=35, swing2LimitDegrees=25, twistLimitDegrees=5}
 {bone1Name=dyn_hair_2_r, bone2Name=dyn_hair_1_r, swing1LimitDegrees=45, swing2LimitDegrees=25, twistLimitDegrees=5}
 {bone1Name=dyn_hair_3_r, bone2Name=dyn_hair_2_r, swing1LimitDegrees=45, swing2LimitDegrees=25, twistLimitDegrees=5}
 {bone1Name=dyn_hair_4_r, bone2Name=dyn_hair_3_r, swing1LimitDegrees=45, swing2LimitDegrees=25, twistLimitDegrees=5}
 {bone1Name=dyn_hair_r, bone2Name=dyn_hair, swing1LimitDegrees=15, swing2LimitDegrees=15, twistLimitDegrees=15}
 {bone1Name=dyn_hair_l, bone2Name=dyn_hair, swing1LimitDegrees=15, swing2LimitDegrees=15, twistLimitDegrees=15}
 {bone1Name=dyn_hair_1_r, bone2Name=dyn_hair_1, swing1LimitDegrees=15, swing2LimitDegrees=15, twistLimitDegrees=15}
 {bone1Name=dyn_hair_1_l, bone2Name=dyn_hair_1, swing1LimitDegrees=15, swing2LimitDegrees=15, twistLimitDegrees=15}
 {bone1Name=dyn_hair_2_r, bone2Name=dyn_hair_2, swing1LimitDegrees=30, swing2LimitDegrees=15, twistLimitDegrees=15}
 {bone1Name=dyn_hair_2_l, bone2Name=dyn_hair_2, swing1LimitDegrees=30, swing2LimitDegrees=15, twistLimitDegrees=15}
 {bone1Name=dyn_hair_3_r, bone2Name=dyn_hair_3, swing1LimitDegrees=30, swing2LimitDegrees=15, twistLimitDegrees=15}
 {bone1Name=dyn_hair_3_l, bone2Name=dyn_hair_3, swing1LimitDegrees=30, swing2LimitDegrees=15, twistLimitDegrees=15}
 {bone1Name=dyn_hair_4_r, bone2Name=dyn_hair_4, swing1LimitDegrees=30, swing2LimitDegrees=15, twistLimitDegrees=15}
 {bone1Name=dyn_hair_4_l, bone2Name=dyn_hair_4, swing1LimitDegrees=30, swing2LimitDegrees=15, twistLimitDegrees=15}
 {bone1Name=dyn_hair_4_f, bone2Name=dyn_hair_2_f, swing1LimitDegrees=15, swing2LimitDegrees=15, twistLimitDegrees=15}
 {
  bone1Name=clavicle_l
  bone2Name=spine_05
  swing1LimitDegrees=45
  swing1Motion=Free
  swing2LimitDegrees=45
  swing2Motion=Free
  twistLimitDegrees=45
  twistMotion=Free
 }
 {
  bone1Name=clavicle_r
  bone2Name=spine_05
  swing1LimitDegrees=45
  swing1Motion=Free
  swing2LimitDegrees=45
  swing2Motion=Free
  twistLimitDegrees=45
  twistMotion=Free
 }
 {
  bone1Name=upperarm_r
  bone2Name=clavicle_r
  swing1LimitDegrees=45
  swing1Motion=Free
  swing2LimitDegrees=45
  swing2Motion=Free
  twistLimitDegrees=45
  twistMotion=Free
 }
 {
  bone1Name=upperarm_l
  bone2Name=clavicle_l
  swing1LimitDegrees=45
  swing1Motion=Free
  swing2LimitDegrees=45
  swing2Motion=Free
  twistLimitDegrees=45
  twistMotion=Free
 }
 {
  bone1Name=spine_04
  bone2Name=spine_03
  swing1LimitDegrees=45
  swing1Motion=Free
  swing2LimitDegrees=45
  swing2Motion=Free
  twistLimitDegrees=45
  twistMotion=Free
 }
 {
  bone1Name=spine_03
  bone2Name=spine_02
  swing1LimitDegrees=45
  swing1Motion=Free
  swing2LimitDegrees=45
  swing2Motion=Free
  twistLimitDegrees=45
  twistMotion=Free
 }
 {
  bone1Name=spine_02
  bone2Name=spine_01
  swing1LimitDegrees=45
  swing1Motion=Free
  swing2LimitDegrees=45
  swing2Motion=Free
  twistLimitDegrees=45
  twistMotion=Free
 }
 {
  bone1Name=spine_01
  bone2Name=pelvis
  swing1LimitDegrees=45
  swing1Motion=Free
  swing2LimitDegrees=45
  swing2Motion=Free
  twistLimitDegrees=45
  twistMotion=Free
 }
 {
  bone1Name=neck_02
  bone2Name=neck_01
  swing1LimitDegrees=45
  swing1Motion=Free
  swing2LimitDegrees=45
  swing2Motion=Free
  twistLimitDegrees=45
  twistMotion=Free
 }
 {bone1Name=dyn_hair_6_f, bone2Name=dyn_hair_4_f, swing1LimitDegrees=20, swing2LimitDegrees=20, twistLimitDegrees=5}
 {bone1Name=dyn_hair_7_f, bone2Name=dyn_hair_1_f, swing1LimitDegrees=20, swing2LimitDegrees=20, twistLimitDegrees=5}
 {
  bone1Name=dyn_hair_3
  bone2Name=spine_03
  swing1LimitDegrees=45
  swing1Motion=Free
  swing2LimitDegrees=45
  swing2Motion=Free
  twistLimitDegrees=45
  twistMotion=Free
 }
 {
  bone1Name=dyn_hair_4
  bone2Name=pelvis
  swing1LimitDegrees=45
  swing1Motion=Free
  swing2LimitDegrees=45
  swing2Motion=Free
  twistLimitDegrees=45
  twistMotion=Free
 }
]
