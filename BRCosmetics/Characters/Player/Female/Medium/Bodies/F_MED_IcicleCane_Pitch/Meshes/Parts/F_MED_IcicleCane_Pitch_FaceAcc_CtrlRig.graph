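# Control Rig for the face accessory part of the female IcicleCane Pitch character, applying a pose-driven correction to hair deform bones as the head moves.
# A spherical pose reader measures the head bone against spine_05, with debug drawing left on.
# Its output weights a Modify Transforms step that offsets deform bones including deform_hair and deform_hair_l.
# The dyn hair and ear bones are children of deform bones, so they inherit any parent correction.
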ControlRigBlueprint /BRCosmetics/Characters/Player/Female/Medium/Bodies/F_MED_IcicleCane_Pitch/Meshes/Parts/F_MED_IcicleCane_Pitch_FaceAcc_CtrlRig
graphs [
 RigVMFunctionLibrary {links [], nodes []}
 RigVMModel {nodes [
   RigUnit_BeginExecution = &RigUnit_BeginExecution()
   n1 = &RigUnit_SphericalPoseReader(
    ActiveRegionScaleFactors=(PositiveWidth=1,NegativeWidth=1,PositiveHeight=2,NegativeHeight=2)
    ActiveRegionSize=0.15
    Debug=(bDrawDebug=true,bDraw2D=False,bDrawLocalAxes=False,DebugScale=25,DebugSegments=20,DebugThickness=0.25)
    DriverAxis=(X=1,Y=0,Z=0)
    DriverItem=(Type=Bone,Name="head")
    FalloffRegionScaleFactors=(PositiveWidth=1,NegativeWidth=1,PositiveHeight=1,NegativeHeight=1)
    FalloffSize=0.7
    FlipHeightScaling=False
    FlipWidthScaling=False
    OptionalParentItem=(Type=Bone,Name="spine_05")
    RotationOffset=(X=-90,Y=0,Z=-40)
    ExecutePin=@RigUnit_BeginExecution.ExecutePin
   )
   > &RigUnit_ModifyTransforms(
    ItemToModify="((Item=(Type=Bone,Name=\"deform_hair\"),Transform=(Rotation=(X=0,Y=0,Z=-0.7071,W=0.7071),Translation=(X=0,Y=0,Z=0),Scale3D=(X=1,Y=2,Z=1))),(Item=(Type=Bone,Name=\"deform_hair_l\"),Transform=(Rotation=(X=0,Y=0,Z=-0.7071,W=0.7071),Translation=(X=0,Y=0,Z=0),Scale ...(truncated)"
    WeightMinimum=0
    Weight=@n1.OutputParam
   )
  ]}
]
hierarchy [
 root {loc=<0 0 0>, rot=<0 0 0 1>}
 pelvis {parent=root, loc=<0 -0.5301 98.0571>, rot=<0.0127 0.707 -0.0127 -0.707>}
 spine_01 {parent=pelvis, loc=<6.1208 -0.9328 0>, rot=<0 0 0.034 -0.9994>}
 spine_02 {parent=spine_01, loc=<6.5997 -0.2095 0>, rot=<0 0 0.0238 -0.9997>}
 spine_03 {parent=spine_02, loc=<6.4914 -0.1237 0>, rot=<0 0 0.0149 -0.9999>}
 spine_04 {parent=spine_03, loc=<6.9598 0.0794 0>, rot=<0 0 -0.0132 -0.9999>}
 spine_05 {parent=spine_04, loc=<6.9987 0 0>, rot=<0 0 -0.0015 -1>}
 neck_01 {parent=spine_05, loc=<11.3138 0.1529 0>, rot=<0 0 -0.2003 -0.9797>}
 neck_02 {parent=neck_01, loc=<4.5627 0 0>, rot=<0 0 0.0504 -0.9987>}
 head {parent=neck_02, loc=<5.3914 0 0>, rot=<0 0 0.115 -0.9934>}
 deform_hair_l {parent=head, loc=<3.657 -4.058 -6.0278>, rot=<0.1553 -0.9876 -0.0036 -0.0229>}
 dyn_hair_l {parent=deform_hair_l, loc=<0 0 0>, rot=<0 0 0 -1>}
 dyn_hair_1_l {parent=dyn_hair_l, loc=<9.0995 -0.0002 0>, rot=<0 0 0.0448 -0.999>}
 dyn_hair_2_l {parent=dyn_hair_1_l, loc=<11.9486 0.0001 0>, rot=<0 0 -0.1372 -0.9905>}
 dyn_hair_3_l {parent=dyn_hair_2_l, loc=<13.298 -0.0002 0>, rot=<0 0 -0.0588 -0.9983>}
 dyn_hair_4_l {parent=dyn_hair_3_l, loc=<14.3489 0.0001 0>, rot=<0 0 -0.0072 -1>}
 deform_hair_r {parent=head, loc=<3.6574 -4.0579 6.0278>, rot=<0.1553 -0.9876 0.0036 0.0229>}
 dyn_hair_r {parent=deform_hair_r, loc=<0 0 0>, rot=<0 0 0 1>}
 dyn_hair_1_r {parent=dyn_hair_r, loc=<9.1003 0 0>, rot=<0 0 0.0448 -0.999>}
 dyn_hair_2_r {parent=dyn_hair_1_r, loc=<11.9482 0 0>, rot=<0 0 -0.1372 -0.9905>}
 dyn_hair_3_r {parent=dyn_hair_2_r, loc=<13.2987 0 0>, rot=<0 0 -0.0588 -0.9983>}
 dyn_hair_4_r {parent=dyn_hair_3_r, loc=<14.3489 0 0>, rot=<0 0 -0.0072 -1>}
 deform_hair {parent=head, loc=<4.6071 -7.0035 0>, rot=<0.1045 -0.9945 0 0>}
 dyn_hair {parent=deform_hair, loc=<0 0 0>, rot=<0 0 0 1>}
 dyn_hair_1 {parent=dyn_hair, loc=<9.1003 0 0>, rot=<0 0 0.0885 -0.9961>}
 dyn_hair_2 {parent=dyn_hair_1, loc=<11.9482 0 0>, rot=<0 0 -0.1372 -0.9905>}
 dyn_hair_3 {parent=dyn_hair_2, loc=<13.2987 0 0>, rot=<0 0 -0.0588 -0.9983>}
 dyn_hair_4 {parent=dyn_hair_3, loc=<14.3489 0 0>, rot=<0 0 -0.0072 -1>}
 deform_ear_l {parent=head, loc=<1.8121 -0.4001 -7.4622>, rot=<0.791 -0.4572 -0.2524 -0.3188>}
 dyn_ear_l {parent=deform_ear_l, loc=<0 0 0>, rot=<0 0 0 1>}
 deform_ear_r {parent=head, loc=<1.8121 -0.4001 7.4622>, rot=<0.791 -0.4572 0.2524 0.3188>}
 dyn_ear_r {parent=deform_ear_r, loc=<0 0 0>, rot=<0 0 0 1>}
 deform_hair_f {parent=head, loc=<11.8577 7.5921 -1.3393>, rot=<0.4351 -0.3989 0.6914 0.4166>}
 dyn_hair_f {parent=deform_hair_f, loc=<0 0 0>, rot=<0 0 0 1>}
 dyn_hair_1_f {parent=dyn_hair_f, loc=<4.5267 0 0>, rot=<0 0 -0.3857 -0.9226>}
 dyn_hair_2_f {parent=dyn_hair_1_f, loc=<7.2363 0.3323 2.6647>, rot=<0.119 0.0327 -0.2866 -0.9501>}
 dyn_hair_3_f {parent=dyn_hair_2_f, loc=<6.2283 0.1295 0.1008>, rot=<-0.0656 -0.1526 0.0131 -0.986>}
 dyn_hair_4_f {parent=dyn_hair_1_f, loc=<7.0308 0.2344 -0.4481>, rot=<0.0101 -0.0419 -0.2341 -0.9712>}
 dyn_hair_5_f {parent=dyn_hair_4_f, loc=<8.7867 0 0>, rot=<-0.0334 -0.0895 0.0516 -0.9941>}
 dyn_hair_6_f {parent=dyn_hair_4_f, loc=<2.0621 0.4069 -2.4752>, rot=<-0.02 0.0664 0.1192 -0.9904>}
 dyn_hair_7_f {parent=dyn_hair_1_f, loc=<2.1523 1.1017 -3.3229>, rot=<-0.0467 -0.0625 -0.0443 -0.996>}
 clavicle_l {parent=spine_05, loc=<5.4831 -0.1053 -3.2946>, rot=<0.0465 -0.7289 -0.009 -0.6829>}
 upperarm_l {parent=clavicle_l, loc=<9.8658 -1.2136 -0.4105>, rot=<0.0164 -0.3245 -0.0269 -0.9454>}
 upperarm_twist_01_l {parent=upperarm_l, loc=<7.2135 -0.1134 0>, rot=<0 0 0 1>}
 deform_upperarm_bk_base_l {parent=upperarm_twist_01_l, loc=<-6.4053 -4.3746 0.314>, rot=<0 0 0 -1>}
 deform_upperarm_bk_l {parent=deform_upperarm_bk_base_l, loc=<0 0 0>, rot=<0 0 0 1>}
 deform_upperarm_fr_base_l {parent=upperarm_twist_01_l, loc=<-5.6312 5.092 0.3354>, rot=<0 0 0 -1>}
 deform_upperarm_fr_l {parent=deform_upperarm_fr_base_l, loc=<0 0 0>, rot=<0 0 0 1>}
 deform_upperarm_up_base_l {parent=upperarm_twist_01_l, loc=<-8.891 0.4825 3.9517>, rot=<0 0 0 -1>}
 deform_upperarm_up_l {parent=deform_upperarm_up_base_l, loc=<0 0 0>, rot=<0 0 0 1>}
 clavicle_r {parent=spine_05, loc=<5.4831 -0.1053 3.2946>, rot=<0.7289 0.0465 -0.6829 0.009>}
 upperarm_r {parent=clavicle_r, loc=<-9.8658 1.2136 0.4105>, rot=<0.0164 -0.3245 -0.0269 -0.9454>}
 upperarm_twist_01_r {parent=upperarm_r, loc=<-7.2135 0.1134 0>, rot=<0 0 0 -1>}
 deform_upperarm_bk_base_r {parent=upperarm_twist_01_r, loc=<6.4053 4.3746 -0.314>, rot=<0 0 0 -1>}
 deform_upperarm_bk_r {parent=deform_upperarm_bk_base_r, loc=<0 0 0>, rot=<0 0 0 1>}
 deform_upperarm_fr_base_r {parent=upperarm_twist_01_r, loc=<5.6312 -5.092 -0.3354>, rot=<0 0 0 -1>}
 deform_upperarm_fr_r {parent=deform_upperarm_fr_base_r, loc=<0 0 0>, rot=<0 0 0 1>}
 deform_upperarm_up_base_r {parent=upperarm_twist_01_r, loc=<8.891 -0.4825 -3.9517>, rot=<0 0 0 -1>}
 deform_upperarm_up_r {parent=deform_upperarm_up_base_r, loc=<0 0 0>, rot=<0 0 0 1>}
]
