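# Physics asset for the front hair of the female medium IcicleCane Pitch character's face accessory part.
# Capsules cover spine, neck, head, clavicles and upper arms, plus a left-arm deform pair, all linked by free angular constraints.
# The dyn_hair chain hangs from deform_hair_f on the head and branches at dyn_hair_1_f and dyn_hair_4_f.
# Hair joints use mostly narrow angular limits, with hair_2 locked to hair_1 and cross-linked to hair_4.
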PhysicsAsset /BRCosmetics/Characters/Player/Female/Medium/Bodies/F_MED_IcicleCane_Pitch/Meshes/Parts/F_MED_IcicleCane_Pitch_FaceAcc_FrontHair_Physics
bodies {
 clavicle_l [{
   shapeType=Capsule
   length=9.3767
   radius=5.6234
   shapeName=clavicle_l_capsule
   center=<4.4774 -0.3141 -1.5367>
   rotation=<rot 82.1518 -0.0003 179.9997>
  }]
 clavicle_r [{
   shapeType=Capsule
   length=8.8107
   radius=6.1894
   shapeName=clavicle_r_capsule
   center=<-4.5172 -0.716 1.2486>
   rotation=<rot 82.1518 -0.0003 -0.0003>
  }]
 deform_hair_f [{shapeType=Capsule, length=1.3071, radius=0.505, shapeName=deform_hair_f_capsule, rotation=<rot 0 -90 0>}]
 deform_upperarm_up_base_l [{shapeType=Capsule, length=0.505, radius=0.505, shapeName=deform_upperarm_up_base_l_capsule, rotation=<rot 0 -90 0>}]
 deform_upperarm_up_l [{shapeType=Capsule, length=4.378, radius=5.622, shapeName=deform_upperarm_up_l_capsule, center=<1.5395 0 -1.4653>, rotation=<rot 90 0 0>}]
 dyn_hair_1_f [{shapeType=Capsule, length=5.8021, radius=3, shapeName=dyn_hair_1_f_capsule, center=<2.6888 -1.4405 0>, rotation=<rot 90 0 0>}]
 dyn_hair_2_f [{shapeType=Capsule, length=6.2994, radius=2.5027, shapeName=dyn_hair_2_f_capsule, center=<2.6888 -1.4405 0>, rotation=<rot 90 0 0>}]
 dyn_hair_3_f [{shapeType=Capsule, length=6.7717, radius=2.0304, shapeName=dyn_hair_3_f_capsule, center=<2.6888 -1.4405 0>, rotation=<rot 90 0 0>}]
 dyn_hair_4_f [{shapeType=Capsule, length=6.2994, radius=2.5027, shapeName=dyn_hair_4_f_capsule, center=<2.6888 -1.4405 0>, rotation=<rot 90 0 0>}]
 dyn_hair_5_f [{shapeType=Capsule, length=8.516, radius=2.0304, shapeName=dyn_hair_5_f_capsule, center=<2.6888 -1.4405 0>, rotation=<rot 90 0 0>}]
 dyn_hair_6_f [{shapeType=Capsule, length=6.6731, radius=1, shapeName=dyn_hair_6_f_capsule, center=<3.2811 0 0>, rotation=<rot 90 0 0>}]
 dyn_hair_7_f [{shapeType=Capsule, length=3.3878, radius=1.1872, shapeName=dyn_hair_7_f_capsule, center=<3.853 0 0>, rotation=<rot 90 0 0>}]
 dyn_hair_f [{shapeType=Capsule, length=5.8021, radius=3, shapeName=dyn_hair_f_capsule, center=<2.6888 -1.4405 0>, rotation=<rot 90 0 0>}]
 head [
  {shapeType=Capsule, length=7.1035, radius=6.9269, center=<3.2006 1.6492 0>, rotation=<rot 90 -34.451 0>}
  {shapeType=Capsule, length=1.8253, radius=7.5486, shapeName=head_capsule, center=<5.8522 0 0>}
 ]
 neck_01 [{shapeType=Capsule, length=8.5081, radius=4.236, shapeName=neck_01_capsule, center=<4.4087 0 0>, rotation=<rot 90 0 0>}]
 neck_02 [{shapeType=Capsule, length=5.2408, radius=4.7592, shapeName=neck_02_capsule, center=<3.0831 0 0>, rotation=<rot 90 0 0>}]
 spine_03 [{shapeType=Capsule, length=11.6973, radius=6.7359, shapeName=spine_03_capsule, center=<1.1021 0 0>, rotation=<rot 0 -90 0>}]
 spine_04 [{shapeType=Capsule, length=10.6277, radius=8.2352, shapeName=spine_04_capsule, center=<1.4616 0 0>}]
 spine_05 [{shapeType=Capsule, length=11.0144, radius=8.9856, shapeName=spine_05_capsule, center=<1.4616 1.3897 0>, rotation=<rot 0 -90 0>}]
 upperarm_l [{shapeType=Capsule, length=25.2216, radius=4.5398, shapeName=upperarm_l_capsule, center=<12.4412 0 0>, rotation=<rot 90 180 0>}]
 upperarm_r [{shapeType=Capsule, length=24.6469, radius=5.1145, shapeName=upperarm_r_capsule, center=<-12.4412 0 0>, rotation=<rot 90 0 0>}]
}
constraints [
 {
  bone1Name=head
  bone2Name=neck_02
  swing1LimitDegrees=45
  swing1Motion=Free
  swing2LimitDegrees=45
  swing2Motion=Free
  twistLimitDegrees=45
  twistMotion=Free
 }
 {
  bone1Name=neck_01
  bone2Name=spine_05
  swing1LimitDegrees=45
  swing1Motion=Free
  swing2LimitDegrees=45
  swing2Motion=Free
  twistLimitDegrees=45
  twistMotion=Free
 }
 {
  bone1Name=spine_05
  bone2Name=spine_04
  swing1LimitDegrees=45
  swing1Motion=Free
  swing2LimitDegrees=45
  swing2Motion=Free
  twistLimitDegrees=45
  twistMotion=Free
 }
 {
  bone1Name=deform_hair_f
  bone2Name=head
  swing1LimitDegrees=45
  swing1Motion=Free
  swing2LimitDegrees=45
  swing2Motion=Free
  twistLimitDegrees=45
  twistMotion=Free
 }
 {bone1Name=dyn_hair_f, bone2Name=deform_hair_f, swing1LimitDegrees=2, swing2LimitDegrees=2, twistLimitDegrees=2}
 {bone1Name=dyn_hair_1_f, bone2Name=dyn_hair_f, swing1LimitDegrees=4, swing2LimitDegrees=4, twistLimitDegrees=2}
 {
  bone1Name=dyn_hair_2_f
  bone2Name=dyn_hair_1_f
  swing1LimitDegrees=35
  swing1Motion=Locked
  swing2LimitDegrees=15
  swing2Motion=Locked
  twistLimitDegrees=5
  twistMotion=Locked
 }
 {bone1Name=dyn_hair_3_f, bone2Name=dyn_hair_2_f, swing1LimitDegrees=35, swing2LimitDegrees=25, twistLimitDegrees=5}
 {bone1Name=dyn_hair_4_f, bone2Name=dyn_hair_1_f, swing1LimitDegrees=35, swing2LimitDegrees=10, twistLimitDegrees=5}
 {bone1Name=dyn_hair_5_f, bone2Name=dyn_hair_4_f, swing1LimitDegrees=35, swing2LimitDegrees=25, twistLimitDegrees=5}
 {bone1Name=dyn_hair_4_f, bone2Name=dyn_hair_2_f, swing1LimitDegrees=15, swing2LimitDegrees=15, twistLimitDegrees=15}
 {
  bone1Name=clavicle_l
  bone2Name=spine_05
  swing1LimitDegrees=45
  swing1Motion=Free
  swing2LimitDegrees=45
  swing2Motion=Free
  twistLimitDegrees=45
  twistMotion=Free
 }
 {
  bone1Name=clavicle_r
  bone2Name=spine_05
  swing1LimitDegrees=45
  swing1Motion=Free
  swing2LimitDegrees=45
  swing2Motion=Free
  twistLimitDegrees=45
  twistMotion=Free
 }
 {
  bone1Name=upperarm_r
  bone2Name=clavicle_r
  swing1LimitDegrees=45
  swing1Motion=Free
  swing2LimitDegrees=45
  swing2Motion=Free
  twistLimitDegrees=45
  twistMotion=Free
 }
 {
  bone1Name=upperarm_l
  bone2Name=clavicle_l
  swing1LimitDegrees=45
  swing1Motion=Free
  swing2LimitDegrees=45
  swing2Motion=Free
  twistLimitDegrees=45
  twistMotion=Free
 }
 {
  bone1Name=spine_04
  bone2Name=spine_03
  swing1LimitDegrees=45
  swing1Motion=Free
  swing2LimitDegrees=45
  swing2Motion=Free
  twistLimitDegrees=45
  twistMotion=Free
 }
 {
  bone1Name=neck_02
  bone2Name=neck_01
  swing1LimitDegrees=45
  swing1Motion=Free
  swing2LimitDegrees=45
  swing2Motion=Free
  twistLimitDegrees=45
  twistMotion=Free
 }
 {bone1Name=dyn_hair_6_f, bone2Name=dyn_hair_4_f, swing1LimitDegrees=20, swing2LimitDegrees=20, twistLimitDegrees=5}
 {bone1Name=dyn_hair_7_f, bone2Name=dyn_hair_1_f, swing1LimitDegrees=20, swing2LimitDegrees=20, twistLimitDegrees=5}
 {
  bone1Name=deform_upperarm_up_base_l
  bone2Name=upperarm_l
  swing1LimitDegrees=45
  swing1Motion=Free
  swing2LimitDegrees=45
  swing2Motion=Free
  twistLimitDegrees=45
  twistMotion=Free
 }
 {
  bone1Name=deform_upperarm_up_l
  bone2Name=deform_upperarm_up_base_l
  swing1LimitDegrees=45
  swing1Motion=Free
  swing2LimitDegrees=45
  swing2Motion=Free
  twistLimitDegrees=45
  twistMotion=Free
 }
]
