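# Animation Blueprint for the IcicleCane Pitch face accessory, simulating hair in the character's neck space.
# It merges parent and body-mesh poses, applies a Control Rig, then feeds chained main and front-hair RigidBody sims or a backpack-specific sim with constraints and squash.
# Disabled physics instead pins hair to spine and pelvis atop the backpack pose; one front bone always follows the head.
# Events detect back bling and update gravity-state parameters.
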AnimBlueprint /BRCosmetics/Characters/Player/Female/Medium/Bodies/F_MED_IcicleCane_Pitch/Meshes/Parts/F_MED_IcicleCane_Pitch_FaceAcc_AnimBP
target_skeleton=&/F_MED_IcicleCane_Pitch_FaceAcc_Skeleton
graphs [
 AnimGraph {nodes [
   n1 = RigidBody!(
    ComponentPose=Local To Component(LocalPose=Use cached pose 'PreRigidPose'())
    Alpha
    OverrideWorldGravity:Vector="0,0,0"
    ExternalForce
    SimSpaceSettings:Sim Space Settings Structure=(WorldAlpha=0.5,VelocityScaleZ=0.4,DampingAlpha=1,MaxLinearVelocity=5000,MaxAngularVelocity=5000,MaxLinearAcceleration=2500,MaxAngularAcceleration=500,ExternalLinearDragV=(X=1,Y=1,Z=1),ExternalLinearVelocity=(X=0,Y=0,Z=0),ExternalAngularVelocity=(X=0,Y=0,Z=0))
   ) -> (Pose) {
    bClampLinearTranslationLimitToRefPose=true
    bOverrideWorldGravity=true
    overridePhysicsAsset=&/F_MED_IcicleCane_Pitch_FaceAcc_Physics
    simulationSpace=BaseBoneSpace
    baseBoneRef {boneName=neck_02}
    simSpaceSettings {
     maxAngularAcceleration=500
     maxAngularVelocity=5000
     maxLinearAcceleration=2500
     maxLinearVelocity=5000
     velocityScaleZ=0.4
     worldAlpha=0.5
     externalLinearDragV=<1 1 1>
    }
   }
   n2 = RigidBody(
    ComponentPose=@n1
    OverrideWorldGravity="0,0,0"
    SimSpaceSettings=(WorldAlpha=0.5,VelocityScaleZ=0.4,DampingAlpha=1,MaxLinearVelocity=5000,MaxAngularVelocity=5000,MaxLinearAcceleration=2500,MaxAngularAcceleration=500,ExternalLinearDragV=(X=1,Y=1,Z=1),ExternalLinearVelocity=(X=0,Y=0,Z=0),ExternalAngularVelocity=(X=0,Y=0,Z=0))
   ) {
    bClampLinearTranslationLimitToRefPose=true
    bOverrideWorldGravity=true
    overridePhysicsAsset=&/F_MED_IcicleCane_Pitch_FaceAcc_FrontHair_Physics
    simulationSpace=BaseBoneSpace
    baseBoneRef {boneName=neck_02}
    simSpaceSettings {
     maxAngularAcceleration=500
     maxAngularVelocity=5000
     maxLinearAcceleration=2500
     maxLinearVelocity=5000
     velocityScaleZ=0.4
     worldAlpha=0.5
     externalLinearDragV=<1 1 1>
    }
   }
   n3 = Blend Poses by bool(
    BlendPose_0=Use cached pose 'CR_Wearing_BP_RBAN'()
    BlendPose_1=Component To Local(ComponentPose=@n2)
    bActiveValue=Get isWearingPackBool() -> (isWearingPackBool:bool)
   ) {blendType=Linear}
   n4 = "Constraint - dyn_hair_1_r (P)"(ComponentPose=Local To Component(LocalPose=Use cached pose 'CR_Wearing_BP_RBAN'()), Alpha=1, ConstraintWeights_0=1) {boneToModify {boneName=dyn_hair_1_r}, constraintSetup [{targetBone {boneName=spine_05}}], constraintWeights [1]}
   n5 = "Constraint - dyn_hair_2_r (P)"(ComponentPose=@n4, Alpha=1, ConstraintWeights_0=1) {boneToModify {boneName=dyn_hair_2_r}, constraintSetup [{targetBone {boneName=spine_05}}], constraintWeights [1]}
   n6 = "Constraint - dyn_hair_3_r (P)"(ComponentPose=@n5, Alpha=1, ConstraintWeights_0=1) {boneToModify {boneName=dyn_hair_3_r}, constraintSetup [{targetBone {boneName=spine_04}}], constraintWeights [1]}
   n7 = "Constraint - dyn_hair_4_r (P)"(ComponentPose=@n6, Alpha=1, ConstraintWeights_0=1) {boneToModify {boneName=dyn_hair_4_r}, constraintSetup [{targetBone {boneName=pelvis}}], constraintWeights [1]}
   n8 = "Constraint - dyn_hair_1 (P)"(ComponentPose=@n7, Alpha=1, ConstraintWeights_0=1) {boneToModify {boneName=dyn_hair_1}, constraintSetup [{targetBone {boneName=spine_05}}], constraintWeights [1]}
   n9 = "Constraint - dyn_hair_2 (P)"(ComponentPose=@n8, Alpha=1, ConstraintWeights_0=1) {boneToModify {boneName=dyn_hair_2}, constraintSetup [{targetBone {boneName=spine_05}}], constraintWeights [1]}
   n10 = "Constraint - dyn_hair_3 (P)"(ComponentPose=@n9, Alpha=1, ConstraintWeights_0=1) {boneToModify {boneName=dyn_hair_3}, constraintSetup [{targetBone {boneName=spine_04}}], constraintWeights [1]}
   n11 = "Constraint - dyn_hair_4 (P)"(ComponentPose=@n10, Alpha=1, ConstraintWeights_0=1) {boneToModify {boneName=dyn_hair_4}, constraintSetup [{targetBone {boneName=pelvis}}], constraintWeights [1]}
   n12 = "Constraint - dyn_hair_1_l (P)"(ComponentPose=@n11, Alpha=1, ConstraintWeights_0=1) {boneToModify {boneName=dyn_hair_1_l}, constraintSetup [{targetBone {boneName=spine_05}}], constraintWeights [1]}
   n13 = "Constraint - dyn_hair_2_l (P)"(ComponentPose=@n12, Alpha=1, ConstraintWeights_0=1) {boneToModify {boneName=dyn_hair_2_l}, constraintSetup [{targetBone {boneName=spine_05}}], constraintWeights [1]}
   n14 = "Constraint - dyn_hair_3_l (P)"(ComponentPose=@n13, Alpha=1, ConstraintWeights_0=1) {boneToModify {boneName=dyn_hair_3_l}, constraintSetup [{targetBone {boneName=spine_04}}], constraintWeights [1]}
   n15 = "Constraint - dyn_hair_4_l (P)"(ComponentPose=@n14, Alpha=1, ConstraintWeights_0=1) {boneToModify {boneName=dyn_hair_4_l}, constraintSetup [{targetBone {boneName=pelvis}}], constraintWeights [1]}
   n16 = Component To Local(ComponentPose=@n15)
   n17 = Get bRigidBodyEnabled()
   n18 = "Constraint - dyn_hair_2_f (P)"(
    ComponentPose=Local To Component(LocalPose=Blend Poses by bool(BlendPose_0=@n3, BlendPose_1=@n16, bActiveValue=@n17) {blendType=Linear})
    Alpha=1
    ConstraintWeights_0=1
   ) {boneToModify {boneName=dyn_hair_2_f}, constraintSetup [{targetBone {boneName=head}}], constraintWeights [1]}
   Output Pose(Result=Component To Local(ComponentPose=@n18)) {layerGroup=None}
   n19 = Layered blend per bone(BasePose=Copy Pose From Mesh() {bUseAttachedParent=true}, BlendPoses_0=Copy Pose From Mesh(SourceMeshComponent=Get Body_Mesh())) {layerSetup [{branchFilters [{blendDepth=1, boneName=clavicle_l}, {blendDepth=1, boneName=clavicle_r}, {boneName=lowerarm_l}, {boneName=lowerarm_r}]}]}
   n20 = Control Rig(Source=@n19) {controlRigAssetReference {blueprintRigClass=&/F_MED_IcicleCane_Pitch_FaceAcc_CtrlRig.F_MED_IcicleCane_Pitch_FaceAcc_CtrlRig_C}}
   Save cached pose 'PreRigidPose'(Pose=@n20) {cacheName=PreRigidPose}
   n21 = RigidBody(
    ComponentPose=Local To Component(LocalPose=Use cached pose 'PreRigidPose'())
    OverrideWorldGravity="0,0,0"
    SimSpaceSettings=(WorldAlpha=0.5,VelocityScaleZ=0.4,DampingAlpha=1,MaxLinearVelocity=5000,MaxAngularVelocity=5000,MaxLinearAcceleration=2500,MaxAngularAcceleration=500,ExternalLinearDragV=(X=1,Y=1,Z=1),ExternalLinearVelocity=(X=0,Y=0,Z=0),ExternalAngularVelocity=(X=0,Y=0,Z=0))
   ) {
    bClampLinearTranslationLimitToRefPose=true
    bOverrideWorldGravity=true
    overridePhysicsAsset=&/F_MED_IcicleCane_Pitch_FaceAcc_Backpack_Physics
    simulationSpace=BaseBoneSpace
    baseBoneRef {boneName=neck_02}
    simSpaceSettings {
     maxAngularAcceleration=500
     maxAngularVelocity=5000
     maxLinearAcceleration=2500
     maxLinearVelocity=5000
     velocityScaleZ=0.4
     worldAlpha=0.5
     externalLinearDragV=<1 1 1>
    }
   }
   n22 = "Constraint - dyn_hair_3_l (P)"(ComponentPose=@n21, Alpha=0, ConstraintWeights_0=1) {boneToModify {boneName=dyn_hair_3_l}, constraintSetup [{targetBone {boneName=spine_03}}], constraintWeights [1]}
   n23 = "Constraint - dyn_hair_3_r (P)"(ComponentPose=@n22, Alpha=0, ConstraintWeights_0=1) {boneToModify {boneName=dyn_hair_3_r}, constraintSetup [{targetBone {boneName=spine_03}}], constraintWeights [1]}
   n24 = "Constraint - dyn_hair_3 (P)"(ComponentPose=@n23, Alpha=0.7, ConstraintWeights_0=1) {boneToModify {boneName=dyn_hair_3}, constraintSetup [{targetBone {boneName=spine_03}}], constraintWeights [1]}
   n25 = "Constraint - dyn_hair_4_l (P)"(ComponentPose=@n24, Alpha=0.7, ConstraintWeights_0=1) {boneToModify {boneName=dyn_hair_4_l}, constraintSetup [{targetBone {boneName=pelvis}}], constraintWeights [1]}
   n26 = "Constraint - dyn_hair_4_r (P)"(ComponentPose=@n25, Alpha=0.7, ConstraintWeights_0=1) {boneToModify {boneName=dyn_hair_4_r}, constraintSetup [{targetBone {boneName=pelvis}}], constraintWeights [1]}
   n27 = "Constraint - dyn_hair_4 (P)"(ComponentPose=@n26, Alpha=0.7, ConstraintWeights_0=1) {boneToModify {boneName=dyn_hair_4}, constraintSetup [{targetBone {boneName=pelvis}}], constraintWeights [1]}
   n28 = "Transform (Modify) Bone - Bone: dyn_hair_3_r"(
    ComponentPose="Transform (Modify) Bone - Bone: dyn_hair_3_l"(ComponentPose=@n27, Alpha=1, Translation="0,0,0", Rotation="0,0,0", Scale="0.7,1,1") {scaleMode=BMM_Additive, scaleSpace=BCS_BoneSpace, boneToModify {boneName=dyn_hair_3_l}, scale {x=0.7}}
    Alpha=1
    Translation="0,0,0"
    Rotation="0,0,0"
    Scale="0.7,1,1"
   ) {scaleMode=BMM_Additive, scaleSpace=BCS_BoneSpace, boneToModify {boneName=dyn_hair_3_r}, scale {x=0.7}}
   Save cached pose 'CR_Wearing_BP_RBAN'(Pose=Component To Local(ComponentPose="Transform (Modify) Bone - Bone: dyn_hair_3"(ComponentPose=@n28, Alpha=1, Translation="0,0,0", Rotation="0,0,0", Scale="0.7,1,1") {scaleMode=BMM_Additive, scaleSpace=BCS_BoneSpace, boneToModify {boneName=dyn_hair_3}, scale {x=0.7}})) {cacheName=CR_Wearing_BP_RBAN}
  ]}
 EventGraph {nodes [
   n1 = "Gravity Override State Function (Parameter Asset) (Nativized)"(
    GameTimeInSeconds=0
    Parameters=/BRCosmetics/Characters/Player/Female/Medium/Bodies/F_MED_LoanFloat/Meshes/Parts/F_MED_LoanFloat_FaceAcc_Parameters
   )
   > Set RBAN Settings Out_FronHair(RBAN Settings Out_FronHair:Fort Rigid Body Anim Node Parameters Structure=@n1.OutRigidBodyAnimNodeParameters) -> (Output_Get:Fort Rigid Body Anim Node Parameters Structure)
   Event BlueprintUpdateAnimation()
   > n2 = "Gravity Override State Function (Parameter Asset) (Nativized)"(
    CharacterPartAnimInstance=Self-Reference()
    GameTimeInSeconds=GetGameTimeInSeconds()
    Parameters=&/F_MED_IcicleCane_Pitch_FaceAcc_Parameters
   )
   > Set Gravity Override(Gravity Override=@n2)
   > Set RBAN Settings Out(RBAN Settings Out=@n2.OutRigidBodyAnimNodeParameters)
   n3 = Event OnCharacterCustomizationLoaded()
   > Set Body_Mesh(Body_Mesh=Get Array Item with EFort Custom Part Type(Array=@n3.CustomMeshParts))
   > Set isWearingPackBool(isWearingPackBool:bool=CheckForBackbling()) -> (Output_Get:bool)
   PrintString!(execute, InString="To String (Name)"(InName=@n2.OutCurrentState), bPrintToScreen, bPrintToLog, TextColor, Duration, Key) -> (then)
  ]}
]
variables [
 Body_Mesh {type=SkeletalMeshComponent}
 Gravity Override {type=Vector}
 is_WearingCape {type=boolean}
 isWearingPackBool {type=boolean}
 RBAN Settings Out {type=FortRigidBodyAnimNodeParameters}
 RBAN Settings Out_FronHair {type=FortRigidBodyAnimNodeParameters}
]
